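; Katana primary attack gameplay ability that feeds input into the Mantis melee system and ends when surface swimming begins.
; On activation, it simulates a Primary "Released" event on the avatar's FortMantisPawnComponent, with a one-tick delay for bot-controlled pawns.
; A parallel sequence branch waits a tick before watching for the MovementMode.SurfaceSwimming tag, so the technique never ends on its starting frame, which can break Mantis.
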(event Ability|EventActivateAbility
  (|CallParentFunction)
  (bind _outfortplayerpawn (CallFunction|GetAvatarasValidFortPlayerPawn)
    (:ValidFortPlayerPawn
      (Utilities|FlowControl|Sequence
        (:then_0
          ;[1 0.7044 0.1557] Delay until the next frame to prevent ending the technique the same frame it was started, which can cause issues for Mantis.
          (Utilities|FlowControl|DelayUntilNextTick)
          (Ability|Tasks|WaitGameplayTagAdd "(TagName=\"MovementMode.SurfaceSwimming\")"
            (:Added
              (Ability|EndAbility))))
        (:then_1
          (Utilities|IsValid (Actor|GetComponentbyClass _outfortplayerpawn "/Script/MantisRuntime.FortMantisPawnComponent")
            (:Is%20Valid
              (if (Pawn|IsBotControlled _outfortplayerpawn)
                (Utilities|FlowControl|DelayUntilNextTick)
                (else
                  _))
              (Mantis|SimulateInputEvent (Actor|GetComponentbyClass _outfortplayerpawn "/Script/MantisRuntime.FortMantisPawnComponent") "Primary" "Released" true false))))))))

(event Ability|EventFailedToActivatePassiveAbility
  (params (ActorInfo :type Gameplay%20Ability%20Actor%20Info%20Structure))
  (|CallParentFunction ActorInfo))

(event Ability|EventOnEndAbility
  (params (bWasCancelled :type Boolean))
  (|CallParentFunction bWasCancelled))
